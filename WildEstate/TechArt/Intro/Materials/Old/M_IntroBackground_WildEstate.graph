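# Old unlit background material for the WildEstate intro, prototyping a screen-space circular mask that animates with intro progress.
# Remapped viewport UVs feed an inverted SphereMask whose center and radius blend between custom-primitive-data End and Start vectors by the MPC Progress raised to a power.
# The aspect correction divides view height by itself, so it has no effect.
# The mask, zoomed-coordinate branch, time-driven sine reroute and background color reach no material output.
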Material /WildEstate/TechArt/Intro/Materials/Old/M_IntroBackground_WildEstate
graph {
 properties {bAutomaticallySetUsageInEditor=false, shadingModel=MSM_Unlit, usedShadingModels=Unlit}
 nodes [
  n1 = MaterialFunctionCall(Float2=ConstantBiasScale(None=ScreenPosition().ViewportUV) {bias=-0.5, scale=2}) {materialFunction=&BreakOutFloat2Components}
  n2 = MaterialFunctionCall(Float2=ViewSize()) {materialFunction=&BreakOutFloat2Components}
  n3 = AppendVector(A=@n1.R, B=Multiply(A=@n1.G, B=Divide(A=@n2.G, B=@n2.G)))
  n4 = VectorParameter() {bUseCustomPrimitiveData=true, parameterName=Sphere Mask End Position + Radius, defaultValue=<rgba 0 0 0 1>}
  n5 = AppendVector(A=@n4.R, B=@n4.R)
  n6 = CollectionParameter() {collection=/WildEstate/TechArt/Intro/MPC_Intro_WildEstate, parameterName=Progress}
  n7 = Power(Base=@n6, Exp=ScalarParameter() {defaultValue=2, parameterName=Zoom Position Power})
  Multiply(A=Add(A=@n3, B=LinearInterpolate(B=@n5, Alpha=@n7)), B=LinearInterpolate(A=@n4.B, B=@n4.B, Alpha=@n6))
  n8 = Sine(None=Multiply(A=Time()) {constB=0.1})
  NamedRerouteDeclaration(None=Saturate(None=ConstantBiasScale(None=@n8))) {name=Progress, nodeColor=<rgba 1 0 0.7734 1>}
  n9 = VectorParameter() {bUseCustomPrimitiveData=true, parameterName=Sphere Mask Start Position + Radius, primitiveDataIndex=4, defaultValue=<rgba 0 0 0 1>}
  n10 = LinearInterpolate(
   A=@n4.RGBA
   B=@n9.RGBA
   Alpha=Power(
    Base=CollectionParameter() {collection=&MPC_Intro_WildEstate, parameterName=Progress}
    Exp=ScalarParameter() {defaultValue=10}
   )
  )
  n11 = ComponentMask(None=@n10) {g=true, r=true}
  OneMinus(None=SphereMask(A=Subtract(A=@n3, B=@n11), B=Constant2Vector(), Radius=ComponentMask(None=@n10) {a=true}) {attenuationRadius=200000})
  TextureCoordinate()
  VectorParameter() {parameterName=Background Color, defaultValue=<rgba 0.0535 0.2874 0.924 1>}
 ]
}
