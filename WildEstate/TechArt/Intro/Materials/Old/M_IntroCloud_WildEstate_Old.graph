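# Old unlit cloud material for the WildEstate intro, with a two-tone color and an animated offset.
# The cloud texture's red channel blends shadow and light color parameters.
# A second blend scales the EndPositionOffset custom primitive data by a Progress reroute.
# Progress comes from a slow time-based sine remapped and clamped to 0–1, while the intro parameter collection's Progress node sits unconnected.
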Material /WildEstate/TechArt/Intro/Materials/Old/M_IntroCloud_WildEstate_Old
graph {
 properties {bAutomaticallySetUsageInEditor=false, shadingModel=MSM_Unlit, usedShadingModels=Unlit}
 nodes [
  CollectionParameter() {collection=/WildEstate/TechArt/Intro/MPC_Intro_WildEstate, parameterName=Progress}
  LinearInterpolate(
   A=VectorParameter() {parameterName=Cloud Shadow, defaultValue=<rgba 0.0634 0.1922 0.4573 1>}.RGB
   B=VectorParameter() {parameterName=CloudLight, defaultValue=<rgba 0.597 0.8694 0.924 1>}.RGB
   Alpha=TextureSample() {texture=/WildEstate/Environment/TestContent/WildEstate_Cloud}.R
  )
  LinearInterpolate(
   B=VectorParameter() {bUseCustomPrimitiveData=true, parameterName=EndPositionOffset, defaultValue=<rgba 0 0 0 1>}.RGB
   Alpha=NamedRerouteUsage()
  )
  n1 = Sine(None=Multiply(A=Time()) {constB=0.1})
  NamedRerouteDeclaration(None=Saturate(None=ConstantBiasScale(None=@n1))) {name=Progress, nodeColor=<rgba 1 0 0.7734 1>}
 ]
}
