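# Translucent unlit material for intro clouds, with per-primitive custom data choosing one of three flipbook frames.
# The frame is color-tinted, and its alpha is remapped, powered and depth-faded against the scene.
# The intro parameter collection's Progress drives an offset lerp and a UV zoom feeding a sphere mask.
# No graph outputs are shown connected, and a time-based sine Progress reroute has no users.
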Material /WildEstate/TechArt/Intro/Materials/Old/M_IntroClouds_WildEstate
graph {
 properties {bAutomaticallySetUsageInEditor=false, blendMode=BLEND_Translucent, shadingModel=MSM_Unlit, usedShadingModels=Unlit}
 nodes [
  LinearInterpolate(
   B=VectorParameter() {bUseCustomPrimitiveData=true, parameterName=EndPositionOffset, defaultValue=<rgba 0 0 0 1>}.RGB
   Alpha=CollectionParameter() {collection=/WildEstate/TechArt/Intro/MPC_Intro_WildEstate, parameterName=Progress}
  )
  n1 = ScalarParameter() {bUseCustomPrimitiveData=true, parameterName=Cloud Index, primitiveDataIndex=4, sliderMax=2}
  n2 = Constant() {r=3}
  n3 = MaterialFunctionCall(
   "Animation  Phase (0-1)"=Divide(A=Round(None=@n1), B=@n2)
   Number of Rows=@n2
   Number of Columns=Constant() {r=1}
   Texture=TextureObject() {texture=/WildEstate/TechArt/Intro/Textures/WildEstate_IntroClouds}
  ) {materialFunction=&FlipBook}
  Multiply(A=@n3.Result, B=VectorParameter() {parameterName=Color Multiplier, defaultValue=<rgba 0.8555 1.0092 1.25 1>}.RGB)
  n4 = InverseLinearInterpolate(
   A=ScalarParameter() {bUseCustomPrimitiveData=true, parameterName=Opacity Falloff Min, primitiveDataIndex=5}
   B=ScalarParameter() {bUseCustomPrimitiveData=true, defaultValue=0.2, parameterName=Opacity Falloff Max, primitiveDataIndex=6}
   Value=@n3.Alpha
  )
  n5 = Subtract(A=SceneDepth().R, B=PixelDepth().R)
  n6 = ScalarParameter() {defaultValue=100, parameterName=Depth Fade Range}
  Multiply(
   A=Power(
    Base=Saturate(None=@n4)
    Exp=ScalarParameter() {bUseCustomPrimitiveData=true, defaultValue=1, parameterName=Opacity Power, primitiveDataIndex=7}
   )
   B=Saturate(None=Divide(A=@n5, B=@n6))
  )
  n7 = Sine(None=Multiply(A=Time()) {constB=0.1})
  NamedRerouteDeclaration(None=Saturate(None=ConstantBiasScale(None=@n7))) {name=Progress, nodeColor=<rgba 1 0 0.7734 1>}
  n8 = VectorParameter() {
   bUseCustomPrimitiveData=true
   parameterName="Mask Zoom Point XY (-1 to 1) + Start/End Zoom ZW"
   primitiveDataIndex=4
   defaultValue=<rgba 0 0 0 1>
  }
  n9 = CollectionParameter() {collection=&MPC_Intro_WildEstate, parameterName=Progress}
  n10 = ScalarParameter() {defaultValue=2, parameterName=Zoom Position Power}
  n11 = Add(
   A=ConstantBiasScale(None=TextureCoordinate()) {bias=-0.5, scale=2}
   B=LinearInterpolate(B=AppendVector(A=@n8.R, B=@n8.R), Alpha=Power(Base=@n9, Exp=@n10))
  )
  n12 = LinearInterpolate(A=@n8.B, B=@n8.B, Alpha=@n9)
  n13 = VectorParameter() {bUseCustomPrimitiveData=true, parameterName=Sphere Mask Position + Radius, primitiveDataIndex=8, defaultValue=<rgba 0 0 0 1>}
  SphereMask(A=Subtract(A=Multiply(A=@n11, B=@n12), B=ComponentMask(None=@n13.RGB) {g=true, r=true}), B=Constant2Vector(), Radius=@n13.B)
 ]
}
