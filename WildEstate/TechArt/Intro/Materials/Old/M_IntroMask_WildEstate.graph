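# Translucent unlit mask material for the WildEstate intro, in an Old folder, that cuts the intro text out of a sphere mask.
# Custom primitive data sets zoom point, zoom and sphere, with MPC Progress animating the UV shift.
# The sphere multiplies the minimum of two offset inverted text-alpha samples, then a scene-depth fade.
# An end-offset lerp, countdown function and sine-based Progress reroute feed nothing.
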Material /WildEstate/TechArt/Intro/Materials/Old/M_IntroMask_WildEstate
graph {
 properties {bAutomaticallySetUsageInEditor=false, blendMode=BLEND_Translucent, shadingModel=MSM_Unlit, usedShadingModels=Unlit}
 nodes [
  LinearInterpolate(
   B=VectorParameter() {bUseCustomPrimitiveData=true, parameterName=EndPositionOffset, defaultValue=<rgba 0 0 0 1>}.RGB
   Alpha=CollectionParameter() {collection=/WildEstate/TechArt/Intro/MPC_Intro_WildEstate, parameterName=Progress}
  )
  MaterialFunctionCall!(UVs, UVPadding, "NumberofDigits(int)") {materialFunction=/Game/Packages/Fortress_SharedMaterials/Base_Material_Functions/BaseCharacterShaderFunctions/MF_CountdownTimer}
  n1 = TextureCoordinate()
  n2 = VectorParameter() {bUseCustomPrimitiveData=true, parameterName="Zoom Point XY (-1 to 1) + Start/End Zoom ZW", primitiveDataIndex=4, defaultValue=<rgba 0 0 0 1>}
  n3 = AppendVector(A=@n2.R, B=@n2.R)
  n4 = CollectionParameter() {collection=&MPC_Intro_WildEstate, parameterName=Progress}
  n5 = Power(Base=@n4, Exp=ScalarParameter() {defaultValue=2, parameterName=Zoom Position Power})
  n6 = Multiply(
   A=Add(A=ConstantBiasScale(None=@n1) {bias=-0.5, scale=2}, B=LinearInterpolate(B=@n3, Alpha=@n5))
   B=LinearInterpolate(A=@n2.B, B=@n2.B, Alpha=@n4)
  )
  n7 = VectorParameter() {bUseCustomPrimitiveData=true, parameterName=Sphere Mask Position + Radius, primitiveDataIndex=8, defaultValue=<rgba 0 0 0 1>}
  n8 = Subtract(A=@n6, B=ComponentMask(None=@n7.RGB) {g=true, r=true})
  n9 = Constant2Vector()
  n10 = ConstantBiasScale(None=@n6)
  n11 = TextureObjectParameter() {parameterName=IntroText, texture=/WildEstate/TechArt/Intro/Textures/WildEstate_IntroText}
  n12 = OneMinus(None=TextureSample(UVs=@n10, Tex=@n11) {samplerSource=SSM_Clamp_WorldGroupSettings, texture=&WildEstate_IntroText}.A)
  n13 = Constant2Vector() {g=-0.005, r=0.005}
  n14 = OneMinus(None=TextureSample(UVs=Add(A=@n10, B=@n13), Tex=@n11) {samplerSource=SSM_Clamp_WorldGroupSettings, texture=&WildEstate_IntroText}.A)
  n15 = Subtract(A=SceneDepth().R, B=PixelDepth().R)
  n16 = ScalarParameter() {defaultValue=100, parameterName=Depth Fade Range}
  Multiply(A=Multiply(A=SphereMask(A=@n8, B=@n9, Radius=@n7.B), B=Min(A=@n12, B=@n14)), B=Saturate(None=Divide(A=@n15, B=@n16)))
  n17 = Sine(None=Multiply(A=Time()) {constB=0.1})
  NamedRerouteDeclaration(None=Saturate(None=ConstantBiasScale(None=@n17))) {name=Progress, nodeColor=<rgba 1 0 0.7734 1>}
 ]
}
