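# Post-process material for the WildEstate intro that composites a scene capture over the rendered scene before bloom for a level transition.
# Capture and mask UVs are scaled by view size over texture size.
# The intro parameter collection's Progress, remapped between transition start and end, fades mask-red areas from capture to live scene, while unmasked areas keep the capture.
# A time-driven sine Progress reroute is left unused.
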Material /WildEstate/TechArt/Intro/Materials/Old/M_IntroPP_WildEstate
graph {
 properties {
  bAutomaticallySetUsageInEditor=false
  bDisablePreExposureScale=true
  blendableLocation=BL_SceneColorBeforeBloom
  materialDomain=MD_PostProcess
  usedShadingModels=Default Lit
 }
 nodes [
  n1 = ViewSize()
  n2 = TextureObjectParameter() {parameterName=SceneCapture}
  n3 = TextureProperty(Tex=@n2)
  n4 = TextureSample(UVs=Multiply(A=TextureCoordinate(), B=Divide(A=@n1, B=@n3)), Tex=@n2)
  n5 = SceneTexture() {sceneTextureId=PPI_PostProcessInput0}
  n6 = InverseLinearInterpolate(
   A=ScalarParameter() {defaultValue=0.8, parameterName=LevelTransitionEndTime}
   B=ScalarParameter() {defaultValue=0.5, parameterName=LevelTransitionStartTime}
   Value=CollectionParameter() {collection=/WildEstate/TechArt/Intro/MPC_Intro_WildEstate, parameterName=Progress}
  )
  n7 = TextureCoordinate()
  n8 = TextureObjectParameter() {parameterName=SceneCaptureMask}
  n9 = Divide(A=ViewSize(), B=TextureProperty(Tex=@n8))
  LinearInterpolate(
   A=@n4.RGB
   B=LinearInterpolate(A=ComponentMask(None=@n5.Color) {b=true, g=true, r=true}, B=@n4.RGB, Alpha=Saturate(None=@n6))
   Alpha=TextureSample(UVs=Multiply(A=@n7, B=@n9), Tex=@n8).R
  )
  n10 = Sine(None=Multiply(A=Time()) {constB=0.1})
  NamedRerouteDeclaration(None=Saturate(None=ConstantBiasScale(None=@n10))) {name=Progress, nodeColor=<rgba 1 0 0.7734 1>}
 ]
}
