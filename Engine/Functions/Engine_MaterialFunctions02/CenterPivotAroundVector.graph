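# Material function that projects each vertex's position onto a local axis vector, giving a pivot point for rotate-about-axis world position offset effects.
# World position is transformed into instance space, then projected with a dot product onto the normalized Vector input, which defaults to local +Z.
# It outputs the projected point as LocalPosition and, transformed back from instance space, as WorldPosition.
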MaterialFunction /Engine/Functions/Engine_MaterialFunctions02/CenterPivotAroundVector
graph {
 properties {
  bExposeToLibrary=true
  description=Returns world position for every vertex as it would be projected against a vector defined by the vector user input. This can be used as the pivot point in a rotate about axis node.
  libraryCategoriesText [Misc, World Position Offset]
 }
 nodes [
  n1 = Constant3Vector() {desc=local vector to create a pivot spot, constant=<rgba 0 0 1 0>}
  n2 = Normalize(VectorInput=FunctionInput(Preview=Normalize(VectorInput=@n1)) {bUsePreviewValueAsDefault=true, inputName=Vector})
  n3 = TransformPosition(None=WorldPosition().XYZ) {desc=could subtract pos here to move axis, transformSourceType=TRANSFORMPOSSOURCE_World, transformType=TRANSFORMPOSSOURCE_Instance}
  n4 = Multiply(A=@n2, B=DotProduct(A=@n2, B=@n3))
  FunctionOutput(None=@n4) {outputName=LocalPosition}
  n5 = TransformPosition(None=@n4) {desc=Projection to eliminate the XY component, transformSourceType=TRANSFORMPOSSOURCE_Instance, transformType=TRANSFORMPOSSOURCE_World}
  FunctionOutput(None=@n5) {outputName=WorldPosition}
 ]
}
