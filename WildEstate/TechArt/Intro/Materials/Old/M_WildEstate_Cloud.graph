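# Unlit cloud material for the WildEstate intro, kept in an Old folder.
# Position-offset math re-bases the mesh from its local axis to yaw toward the camera around world Z, then bends vertices horizontally by camera-distance difference from the instance origin.
# Texture red blends collection shadow and light colors, with branches for time-of-day tint and eye-adaptation-compensated storm tint, alongside a safe-zone progress value.
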Material /WildEstate/TechArt/Intro/Materials/Old/M_WildEstate_Cloud
graph {
 properties {bAutomaticallySetUsageInEditor=false, bTangentSpaceNormal=false, shadingModel=MSM_Unlit, usedShadingModels=Unlit}
 nodes [
  n1 = ComponentMask(None=CameraPositionWS()) {g=true, r=true}
  n2 = WorldPosition()
  n3 = ObjectPositionWS()
  n4 = Normalize(VectorInput=Transform(None=Constant3Vector() {constant=<rgba -1 0 0 0>}) {transformSourceType=TRANSFORMSOURCE_Local})
  n5 = Constant3Vector() {constant=<rgba 0 0 1 0>}
  n6 = Add(
   A=MaterialFunctionCall(VectorToTransform=Subtract(A=@n2.XYZ, B=@n3), BasisX=@n4, BasisY=CrossProduct(A=@n4, B=@n5), BasisZ=@n5) {materialFunction=&Transform3x3Matrix}.Transformed Vector
   B=@n3
  )
  n7 = WorldPosition()
  n8 = ObjectPositionWS()
  n9 = Subtract(A=Add(A=Subtract(A=@n6, B=@n2.XYZ), B=@n7.XYZ), B=@n8)
  n10 = ComponentMask(None=CameraPositionWS()) {g=true, r=true}
  n11 = ComponentMask(None=ObjectPositionWS()) {g=true, r=true}
  n12 = AppendVector(A=Normalize(VectorInput=Subtract(A=@n10, B=@n11)), B=Constant())
  n13 = Constant3Vector() {constant=<rgba 0 0 1 0>}
  n14 = CrossProduct(A=@n12, B=@n13)
  n15 = Subtract(
   A=Add(
    A=MaterialFunctionCall(VectorToTransform=@n9, BasisX=@n12, BasisY=@n14, BasisPosition=@n13) {materialFunction=&Transform3x3Matrix}.Transformed Vector
    B=@n8
   )
   B=@n7.XYZ
  )
  n16 = ComponentMask(None=Add(A=WorldPosition().XYZ, B=@n15)) {g=true, r=true}
  n17 = TransformPosition(None=Constant3Vector()) {transformSourceType=TRANSFORMPOSSOURCE_Instance, transformType=TRANSFORMPOSSOURCE_World}
  n18 = Subtract(A=Distance(A=@n1, B=@n16), B=Distance(A=ComponentMask(None=@n17) {g=true, r=true}, B=@n1))
  n19 = Normalize(VectorInput=AppendVector(A=Subtract(A=@n1, B=@n16), B=Constant()))
  Add(A=Multiply(A=Multiply(A=@n18, B=@n19), B=ScalarParameter() {defaultValue=1.9452, parameterName=Bend}) {constB=4}, B=@n15)
  Constant()
  MaterialFunctionCall!(Vector) {materialFunction=/Engine/Functions/Engine_MaterialFunctions02/CenterPivotAroundVector}
  n20 = TextureSample() {texture=/WildEstate/Environment/TestContent/WildEstate_Cloud}
  n21 = LinearInterpolate(
   A=ComponentMask(None=CollectionParameter() {collection=&FortniteMaterialParameters, parameterName=CloudsShadow}) {b=true, g=true, r=true}
   B=ComponentMask(None=CollectionParameter() {collection=&FortniteMaterialParameters, parameterName=CloudsLight}) {b=true, g=true, r=true}
   Alpha=@n20.R
  )
  Multiply(
   A=MaterialFunctionCall(
    Color_Morning=VectorParameter() {parameterName=Color_Bottom_Morning, defaultValue=<rgba 1 0.7511 0.4896 1>}.RGB
    Color_Day=VectorParameter() {parameterName=Color_Bottom_Day, defaultValue=<rgba 1 1 1 1>}.RGB
    Color_Evening=VectorParameter() {parameterName=Color_Bottom_Evening, defaultValue=<rgba 0.6198 0.4161 0.3118 1>}.RGB
    Color_Night=VectorParameter() {parameterName=Color_Bottom_Night, defaultValue=<rgba 0.0503 0.02 0.1 1>}.RGB
   ) {materialFunction=&Color_Time_of_Day}.Result
   B=@n21
  )
  n22 = VectorParameter() {parameterName=StormColorTint, defaultValue=<rgba 0.593 0.0046 0.6458 1>}
  n23 = VectorParameter() {parameterName=StormColorTint2, defaultValue=<rgba 0.3035 0.026 1 1>}
  n24 = Multiply(A=NamedRerouteUsage(), B=ScalarParameter() {defaultValue=10, parameterName=ColorChangeStromMult})
  Multiply(
   A=LinearInterpolate(A=Constant3Vector() {constant=<rgba 1 1 1 0>}, B=LinearInterpolate(A=@n22.RGB, B=@n23.RGB, Alpha=@n20.R), Alpha=Saturate(None=@n24))
   B=EyeAdaptationInverse(LightValueInput=@n21).EyeAdaptationInverse
  )
  n25 = OneMinus(None=Saturate(None=CollectionParameter() {collection=/Game/Athena/SafeZone/MPC_SafeZoneClouds, parameterName=SafeZoneScale}))
  NamedRerouteDeclaration(None=Saturate(None=Power(Base=@n25))) {name=ProgressNormalized, nodeColor=<rgba 1 0 0.375 1>}
  ScalarParameter() {defaultValue=1, parameterName=Opacity}
  VectorParameter() {parameterName=CloudColor, defaultValue=<rgba 3 3 3 1>}
  VectorParameter() {parameterName=ShadowColor, defaultValue=<rgba 0.6739 1.2138 2.5 1>}
 ]
}
